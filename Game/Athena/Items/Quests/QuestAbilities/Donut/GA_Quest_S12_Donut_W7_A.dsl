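; Quest gameplay ability for the Season 12 Donut Week 7 challenge that tracks picking up a specific pistol item.
; On non-server instances it waits for quests to be ready, then binds to inventory pickups if the local player has either quest and either objective is incomplete.
; A matching pickup sends two client stat events, calls ServerStatEvent to send the tagged PickupPistols custom stat, then unbinds, so it fires once.
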(event Ability|EventActivateAbility
  (bind _owningplayerpawn (Variables|GAVars|Data|GetOwningPlayerPawn))
  (bind _returnvalue (Fortnite|GetFortPlayerControllerfromActor _owningplayerpawn))
  (|CallParentFunction)
  (if (not (Networking|IsServer))
    (Quests|CurrentQuestsReadyAsync (Variables|GAVars|Data|GetOwningPlayerQuestManager) (Utilities|Array|MakeArray 0)
      (:OnCurrentQuestsReady
        (bind _questvalid (Quest|DoesPlayerHaveQuest _returnvalue 0 "Quest_S12_Donut_W7_A_0"))
        (bind _questvalid_2 (Quest|DoesPlayerHaveQuest _returnvalue 0 "Quest_S12_Donut_W7_A_1"))
        (bind (_playercontrollerout _completedquestobjective) (Quest|HasPlayerCompletedQuestObjective _returnvalue 0 "Quest_S12_Donut_W7_A_0"))
        (bind (_playercontrollerout_2 _completedquestobjective_2) (Quest|HasPlayerCompletedQuestObjective _returnvalue 0 "Quest_S12_Donut_W7_A_1"))
        (if (and (or _questvalid _questvalid_2) (or (not _completedquestobjective) (not _completedquestobjective_2)))
          (if (Utilities|Equal(Object) (Fortnite|GetFirstLocalFortPlayerController true) (Fortnite|GetFortPlayerControllerfromActor (Variables|GAVars|Data|GetOwningPlayerPawn)))
            (Inventory|BindEventtoOnItemPickedUp (BlueprintContext|GetFortInventoryContext) (delegate Custom|ItemPickedUp))))))))

(event Ability|EventFailedToActivatePassiveAbility
  (params (ActorInfo :type Gameplay%20Ability%20Actor%20Info%20Structure)))

(event Custom|ItemPickedUp
  (params (NewItem :type Fort%20World%20Item%20Object%20Reference) (Count :type Integer))
  (if (Utilities|Equal(Object) (Item|GetItemDefinition NewItem) (Variables|Default|GetPistolItemDef))
    (Quests|SendClientStatEvent (Variables|GAVars|Data|GetOwningPlayerQuestManager))
    (Quests|SendClientStatEvent (Variables|GAVars|Data|GetOwningPlayerQuestManager))
    (CallFunction|ServerStatEvent)
    (Inventory|UnbindEventfromOnItemPickedUp (BlueprintContext|GetFortInventoryContext) (delegate Custom|ItemPickedUp))))

(event Custom|ServerStatEvent
  (Quest|SendComplexCustomStatEvent (Variables|GAVars|Data|GetOwningPlayerQuestManager) 0 "(GameplayTags=)" "(GameplayTags=((TagName=\"Athena.Quests.Donut.PickupPistols\")))" 2))
